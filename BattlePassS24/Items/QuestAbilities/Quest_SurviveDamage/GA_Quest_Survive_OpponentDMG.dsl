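; Quest gameplay ability that reports a damage-survival stat when the owner is hit by an opposing player.
; Activation filters for non-teammate, player-tagged, non-environment damage, and requires the GameOn context tag or ends immediately.
; Chained listeners for received, shielded and overshielded damage apply the same filter and feed a shared 30-second retriggerable delay, restarted by further hits.
; Expiry sends the quest stat and ends the ability, with no death check in this graph.
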(event Ability|EventActivateAbilityFromEvent
  (params (EventData :type Gameplay%20Event%20Data%20Structure%20%28by%20ref%29))
  (bind (_outsourcetags _outcontexttags) (Quests|GetSourceandContextTags (Variables|GAVars|Data|GetOwningPlayerQuestManager)))
  (|CallParentFunction EventData)
  (if (GameplayTags|HasTag _outcontexttags "(TagName=\"Athena.GameOn\")")
    (bind (_asynctaskproxy _payload) (Ability|Tasks|WaitGameplayEvent "(TagName=\"Event.Damage.Received\")")
      (:then
        (bind (_asynctaskproxy_2 _payload_2) (Ability|Tasks|WaitGameplayEvent "(TagName=\"Event.Damage.Shielded\")")
          (:then
            (bind (_asynctaskproxy_3 _payload_3) (Ability|Tasks|WaitGameplayEvent "(TagName=\"Event.Damage.Overshielded\")")
              (:then
                _)
              (:EventReceived
                (bind (_payload_eventtag _payload_instigator _payload_target _payload_optionalobject _payload_optionalobject2 _payload_contexthandle _payload_instigatortags) (Utilities|Struct|BreakGameplayEventData _payload_3))
                (bind _asfort_player_pawn (Utilities|Casting|CastToFortPlayerPawn _payload_target))
                (if (Math|Boolean|ANDBoolean (not (Team|OnSameTeam _payload_instigator _asfort_player_pawn)) (not (GameplayTags|HasAnyTags _payload_instigatortags (Variables|Default|GetT_EnvironmentDMG))) (GameplayTags|HasAnyTags _payload_instigatortags (Variables|Default|GetT_player)))
                  _))))
          (:EventReceived
            (bind (_payload_eventtag_2 _payload_instigator_2 _payload_target_2 _payload_optionalobject_2 _payload_optionalobject2_2 _payload_contexthandle_2 _payload_instigatortags_2) (Utilities|Struct|BreakGameplayEventData _payload_2))
            (bind _asfort_player_pawn_2 (Utilities|Casting|CastToFortPlayerPawn _payload_target_2))
            (if (Math|Boolean|ANDBoolean (not (Team|OnSameTeam _payload_instigator_2 _asfort_player_pawn_2)) (not (GameplayTags|HasAnyTags _payload_instigatortags_2 (Variables|Default|GetT_EnvironmentDMG))) (GameplayTags|HasAnyTags _payload_instigatortags_2 (Variables|Default|GetT_player)))
              _))))
      (:EventReceived
        (bind (_payload_eventtag_3 _payload_instigator_3 _payload_target_3 _payload_optionalobject_3 _payload_optionalobject2_3 _payload_contexthandle_3 _payload_instigatortags_3) (Utilities|Struct|BreakGameplayEventData _payload))
        (bind _asfort_player_pawn_3 (Utilities|Casting|CastToFortPlayerPawn _payload_target_3))
        (if (Math|Boolean|ANDBoolean (not (Team|OnSameTeam _payload_instigator_3 _asfort_player_pawn_3)) (not (GameplayTags|HasAnyTags _payload_instigatortags_3 (Variables|Default|GetT_EnvironmentDMG))) (GameplayTags|HasAnyTags _payload_instigatortags_3 (Variables|Default|GetT_player)))
          _)))
    (Utilities|FlowControl|RetriggerableDelay 3e+01)
    (Quest|SendComplexCustomStatEvent (Variables|GAVars|Data|GetOwningPlayerQuestManager) 0 (Variables|Default|GetT_TakeDMG_Survive) "(GameplayTags=)" 1)
    (Ability|EndAbility)
    (else
      (Ability|EndAbility))))

(fn ShouldAbilityRespondToEvent
  (params (ActorInfo :type Gameplay%20Ability%20Actor%20Info%20Structure) (Payload_EventTag :type Gameplay%20Tag%20Structure) (Payload_Instigator :type Actor%20Object%20Reference) (Payload_Target :type Actor%20Object%20Reference) (Payload_OptionalObject :type Object%20Reference) (Payload_OptionalObject2 :type Object%20Reference) (Payload_ContextHandle :type Gameplay%20Effect%20Context%20Handle%20Structure) (Payload_InstigatorTags :type Gameplay%20Tag%20Container%20Structure) (Payload_TargetTags :type Gameplay%20Tag%20Container%20Structure) (Payload_EventMagnitude :type Float%20%28single-precision%29) (Payload_TargetData :type Gameplay%20Ability%20Target%20Data%20Handle%20Structure) (Payload_InstancedEventData :type Instanced%20Struct%20Structure))
  (bind _asfort_player_pawn (Utilities|Casting|CastToFortPlayerPawn Payload_Target)
    (:then
      (return (Math|Boolean|ANDBoolean (not (Team|OnSameTeam _asfort_player_pawn Payload_Instigator)) (not (GameplayTags|HasAnyTags Payload_InstigatorTags (Variables|Default|GetT_EnvironmentDMG))) (GameplayTags|HasAnyTags Payload_InstigatorTags (Variables|Default|GetT_player)))))
    (:CastFailed
      (return false))))
